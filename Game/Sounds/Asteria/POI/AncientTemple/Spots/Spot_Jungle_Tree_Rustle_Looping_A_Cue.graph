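# Looping ambient spot cue that plays intermittent tree rustles, named for the jungle around the Ancient Temple POI.
# A loop waits a random delay, then a quality switch feeds a random pick between borrowed Cherry Blossom and Knights forest rustle pools, each with pitch and volume modulation.
# Only the first two quality slots are connected; the other two are empty.
# Tree-rustle attenuation applies, and a concurrency override allows one instance.
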SoundCue /Game/Sounds/Asteria/POI/AncientTemple/Spots/Spot_Jungle_Tree_Rustle_Looping_A_Cue
root=Looping_0
nodes [
 {id=Delay_0, children [QualityLevel_1], properties {delayMax=12, delayMin=6}}
 {id=Looping_0, children [Delay_0]}
 {id=Modulator_0, children [Random_2], properties {pitchMax=1.1, pitchMin=1, volumeMax=1, volumeMin=0.25}}
 {id=Modulator_1, children [Random_0], properties {pitchMax=1.1, pitchMin=1, volumeMax=1.1, volumeMin=0.5}}
 {id=QualityLevel_1, children [Random_3,Random_3,null,null]}
 {id=Random_0, children [WavePlayer_0,WavePlayer_1,WavePlayer_2,WavePlayer_3,WavePlayer_4,WavePlayer_5], properties {weights [1,1,1,1,1,1]}}
 {
  id=Random_2
  children [
   WavePlayer_14
   WavePlayer_15
   WavePlayer_16
   WavePlayer_17
   WavePlayer_18
   WavePlayer_19
   WavePlayer_20
   WavePlayer_21
   WavePlayer_22
   WavePlayer_23
  ]
  properties {weights [1,1,1,1,1,1,1,1,1,1]}
 }
 {id=Random_3, children [Modulator_1,Modulator_0], properties {weights [1,1]}}
 {id=WavePlayer_0, properties {soundWaveAssetPtr=/Game/Sounds/Asteria/POI/Cherry_Blossom_Forest/Spots/Spot_CherryBlossom_Tree_Rustle_06}}
 {id=WavePlayer_1, properties {soundWaveAssetPtr=/Game/Sounds/Asteria/POI/Cherry_Blossom_Forest/Spots/Spot_CherryBlossom_Tree_Rustle_01}}
 {id=WavePlayer_14, properties {soundWaveAssetPtr=/Game/Sounds/Asteria/Biomes/Knights/Spots/Spot_Knights_Forest_Tree_Rustle_01}}
 {id=WavePlayer_15, properties {soundWaveAssetPtr=/Game/Sounds/Asteria/Biomes/Knights/Spots/Spot_Knights_Forest_Tree_Rustle_02}}
 {id=WavePlayer_16, properties {soundWaveAssetPtr=/Game/Sounds/Asteria/Biomes/Knights/Spots/Spot_Knights_Forest_Tree_Rustle_03}}
 {id=WavePlayer_17, properties {soundWaveAssetPtr=/Game/Sounds/Asteria/Biomes/Knights/Spots/Spot_Knights_Forest_Tree_Rustle_04}}
 {id=WavePlayer_18, properties {soundWaveAssetPtr=/Game/Sounds/Asteria/Biomes/Knights/Spots/Spot_Knights_Forest_Tree_Rustle_05}}
 {id=WavePlayer_19, properties {soundWaveAssetPtr=/Game/Sounds/Asteria/Biomes/Knights/Spots/Spot_Knights_Forest_Tree_Rustle_06}}
 {id=WavePlayer_2, properties {soundWaveAssetPtr=/Game/Sounds/Asteria/POI/Cherry_Blossom_Forest/Spots/Spot_CherryBlossom_Tree_Rustle_02}}
 {id=WavePlayer_20, properties {soundWaveAssetPtr=/Game/Sounds/Asteria/Biomes/Knights/Spots/Spot_Knights_Forest_Tree_Rustle_07}}
 {id=WavePlayer_21, properties {soundWaveAssetPtr=/Game/Sounds/Asteria/Biomes/Knights/Spots/Spot_Knights_Forest_Tree_Rustle_08}}
 {id=WavePlayer_22, properties {soundWaveAssetPtr=/Game/Sounds/Asteria/Biomes/Knights/Spots/Spot_Knights_Forest_Tree_Rustle_09}}
 {id=WavePlayer_23, properties {soundWaveAssetPtr=/Game/Sounds/Asteria/Biomes/Knights/Spots/Spot_Knights_Forest_Tree_Rustle_10}}
 {id=WavePlayer_3, properties {soundWaveAssetPtr=/Game/Sounds/Asteria/POI/Cherry_Blossom_Forest/Spots/Spot_CherryBlossom_Tree_Rustle_03}}
 {id=WavePlayer_4, properties {soundWaveAssetPtr=/Game/Sounds/Asteria/POI/Cherry_Blossom_Forest/Spots/Spot_CherryBlossom_Tree_Rustle_04}}
 {id=WavePlayer_5, properties {soundWaveAssetPtr=/Game/Sounds/Asteria/POI/Cherry_Blossom_Forest/Spots/Spot_CherryBlossom_Tree_Rustle_05}}
]
properties {
 attenuationSettings=/Game/Sounds/Attenuation/Fort_TileAmbient_OneShot_Tree_Rustle_Att
 bEnableBusSends=false
 bOverrideConcurrency=true
 maxDistance=6850
 soundClassObject=&Fort_Ambient_OneShots_Outdoors
 volumeMultiplier=0.601
 concurrencyOverrides {maxCount=1, voiceStealReleaseTime=2}
 preEffectBusSends [{}]
}
